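; Pine tree foliage Blueprint for the Asteria environment's boreal conifers, with a custom Rescale event that randomizes the actor's scale.
; Rescale splits the RandomScaleRange 2D vector into minimum and maximum, then rolls a separate random value for X, Y and Z, so scaling is non-uniform.
; The construction script only calls its parent; BeginPlay, overlap and Tick have no active logic, and nothing here calls Rescale.
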(event EventBeginPlay
  (|CallParentFunction) ; This node is disabled and will not be called.\nDrag off pins to build functionality.
)

(event Collision|EventActorBeginOverlap
  (params (OtherActor :type Actor%20Object%20Reference)))

(event EventTick
  (params (DeltaSeconds :type Float%20%28single-precision%29)))

(event Custom|Rescale
  (bind (_x _y) (Math|Vector2D|BreakVector2D (Variables|GetRandomScaleRange)))
  (Transformation|SetActorScale3D self (Math|Vector|MakeVector (Math|Random|RandomFloatinRange _x _y) (Math|Random|RandomFloatinRange _x _y) (Math|Random|RandomFloatinRange _x _y))))

(fn ConstructionScript
  (|CallParentFunction))
